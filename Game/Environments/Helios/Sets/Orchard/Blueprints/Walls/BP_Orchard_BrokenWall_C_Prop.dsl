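; Broken orchard wall prop (variant C) for the Helios environment, set up as a simple destructible prop.
; It exposes its single static mesh component as the destructible component and returns a fixed Wood entry for it, with no attachment or slot.
; It reports that the source actor should not be kept, and its construction script is empty.
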(fn GetSimpleDestructiblePropComponents
  (return (Utilities|Array|MakeArray (Variables|Debug|GetStaticMeshComponent))))

(fn GetSimpleDestructiblePropEntry
  (params (SourceIndex :type Integer) (SourceComponent :type Static%20Mesh%20Component%20Object%20Reference))
  (return 0 0 "(Name=\"\")" 0 "(CurveTable=None,RowName=\"\")" "Wood" "Wood" "ATTACH_None" "SLOT_None" false false false))

(fn ShouldKeepSourceActor
  (return false))

(fn ConstructionScript)
